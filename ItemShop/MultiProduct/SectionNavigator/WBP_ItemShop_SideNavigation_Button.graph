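# Item Shop side navigation button, a child of the UIKit side navigation button, that adjusts its sizing for compact UI.
# On construct and on input method change, it passes the current input type to SetCompactUIStyle.
# When compact UI is in use, that function picks cached content size, mobile minimum width and text wrap width based on whether the input is gamepad.
# The wrap width is also applied to the text inside the inherited Block_Text.
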WidgetBlueprint /ItemShop/MultiProduct/SectionNavigator/WBP_ItemShop_SideNavigation_Button
graphs [
 EventGraph {nodes [
   Event Construct()
   > SetCompactUIStyle!(execute, self, CurrentInputType:ECommonInputType Enum=Get CurrentInputType() -> (CurrentInputType:ECommonInputType Enum)) -> (then) {is=title K2Node_CallFunction}
   n1 = Event On Input Method Changed!() -> (OutputDelegate, then, CurrentInputType:ECommonInputType Enum) {is=title K2Node_Event}
   > SetCompactUIStyle(CurrentInputType=@n1.CurrentInputType)
  ]}
 SetCompactUIStyle {nodes [
   n1 = SetCompactUIStyle!() -> (then, CurrentInputType:ECommonInputType Enum) {is=title K2Node_FunctionEntry}
   > Branch(Condition=IsUsingCompactUI())
   n2 = "Equal (Enum)"(A=@n1, B=Gamepad)
   > Set CachedContentSize(CachedContentSize:Vector 2D Structure=Select(Option 0:Vector 2D Structure=(X=420,Y=147), Option 1:Vector 2D Structure=(X=540,Y=147), Index=@n2) -> (ReturnValue:Vector 2D Structure)) -> (Output_Get:Vector 2D Structure)
   > Set MobileMinWidth(MobileMinWidth:int=SelectInt(A=360, B=440, bPickA=@n2)) -> (Output_Get:int=0)
   > n3 = Set WrapTextAt(WrapTextAt:int=SelectInt(A=320, B=400, bPickA=@n2)) -> (Output_Get:int=0)
   > SetWrapTextWidth!(
    execute
    self:Common Text Object Reference=Get TextBlock!(self:WBP UIKit Block Text Object Reference=Get Block_Text() -> (Block_Text:WBP UIKit Block Text Object Reference)) -> (TextBlock:Common Text Object Reference)
    InWrapTextAt:int=@n3
   ) -> (then)
  ]}
]
variables []
widgets {
 info {
  inheritedWidgetCount=0
  parentClass=/FortUILibrary/UIKit/Buttons/TabButtons/WBP_UIKit_SideNavigation_Button.WBP_UIKit_SideNavigation_Button_C
  rootWidgetClass=null
  widgetCount=0
 }
 widgets []
}
