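# Item shop side navigation button widget for section entries in the multi-product section navigator, as its name and folder indicate.
# It is a child Widget Blueprint of WBP_ItemShop_SideNavigation_Button.
# It adds no widgets, variables or event graph logic of its own, so any layout and behavior come from the parent class.
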WidgetBlueprint /ItemShop/MultiProduct/SectionNavigator/WBP_ItemShop_SideNavigation_Section_Button
graphs [EventGraph {links [], nodes []}]
variables []
widgets {
 info {
  inheritedWidgetCount=0
  parentClass=&/WBP_ItemShop_SideNavigation_Button.WBP_ItemShop_SideNavigation_Button_C
  rootWidgetClass=null
  widgetCount=0
 }
 widgets []
}
